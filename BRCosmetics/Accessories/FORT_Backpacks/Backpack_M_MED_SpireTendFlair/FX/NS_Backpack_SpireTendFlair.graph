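# Niagara effect for the SpireTendFlair back bling cosmetic, showing blue-tinted chunk, droplet and smoke visuals that change with User.ElimCount.
# Burst emitters throw mesh chunks, sprite droplets and smoke wisps, using masks and switches on that count to vary amount, origin, velocity and gravity.
# Rate-based emitters add dripping droplets and drifting smoke detail, with per-count spawn probabilities that silence both at higher counts.
# Time-of-day functions adjust particle colors.
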NiagaraSystem /BRCosmetics/Accessories/FORT_Backpacks/Backpack_M_MED_SpireTendFlair/FX/NS_Backpack_SpireTendFlair
dependencies {
 usedDataInterfaces [
  &NiagaraDataInterfaceMeshRendererInfo
  &NiagaraDataInterfaceCurve
  &NiagaraDataInterfaceSpriteRendererInfo
  &NiagaraDataInterfaceColorCurve
  &NiagaraDataInterfaceArrayFloat
  &NiagaraDataInterfaceArrayFloat3
  &NiagaraDataInterfaceVector2DCurve
 ]
 usedDynamicInputs [&Vector2DFromCurve, &RandomRangeVector2D]
 usedModules [
  &self:SystemSpawnScript.SystemScriptSource.SystemScriptGraph.NiagaraNodeAssignment_0.SetVariables_#1
  &SystemState
 ]
}
emitters {
 Chunk {
  emitterUpdateScript [
   &EmitterState()
   &SpawnBurst_Instantaneous(
    Spawn Count=/Niagara/DynamicInputs/Helpers/MaskInt(
     Mask Source:NiagaraInt32=1
     Mask Parameter:NiagaraInt32=$User.ElimCount
     Mask:NiagaraInt32=0
     Masked Int:NiagaraInt32=&RandomRangeInt(Minimum=12, Maximum=18)
     Passthrough Int:NiagaraInt32=&RandomRangeInt(Minimum=4, Maximum=8)
    )
    Use Spawn Probability=false
    UseLoopCountLimit=false
   )
   &SpawnBurst_Instantaneous(
    Spawn Count=&MaskInt(
     Mask Source:NiagaraInt32=1
     Mask Parameter:NiagaraInt32=$User.ElimCount
     Mask:NiagaraInt32=0
     Masked Int:NiagaraInt32=&RandomRangeInt(Minimum=7, Maximum=10)
     Passthrough Int:NiagaraInt32=&RandomRangeInt(Minimum=3, Maximum=5)
    )
    Spawn Group=1
    Use Spawn Probability=false
    UseLoopCountLimit=false
   ) {module=SpawnBurst_Instantaneous001}
  ]
  particleSpawnScript [
   /SetVariables(Particles.VisibilityTag=&SpawnGroup()) {isSetParameters=true}
   &InitializeParticle(
    Color=&Multiply_LinearColorByFloat(
     Linear Color=&RandomRangeLinearColor(Minimum=<rgba 0.1802 0.2337 0.3802 1>, Maximum=<rgba 0.3864 0.4342 0.5647 1>)
     Float=&FNTimeOfDayScalar(Night Scale=1.4, Morning Scale=1, Evening Scale=0)
    )
    Lifetime=&MaskFloat(
     Mask Source=1
     Mask Parameter=$User.ElimCount
     Mask=0
     Masked Float=&RandomRangeFloat(Minimum=0.275, Maximum=0.5)
     Passthrough Float=&RandomRangeFloat(Minimum=0.325, Maximum=0.6)
    )
    Mesh Renderer Array Visibility Mode=::NewEnumerator3
    Mesh Renderer Info=DataInterface() {propertyValues {MeshRenderer=&self:Chunk_0.NiagaraMeshRendererProperties_0}}
    Mesh Renderer Random Seed:NiagaraInt32=&RandomRangeInt(Minimum=0, Maximum=4)
    Mesh Scale Mode=::NewEnumerator3
    Mesh Uniform Scale=&Multiply_Float(
     A=&MaskFloatBySpawnGroup(
      Masked Float=&UniformAOrBFloat(A=0.05, B=0.125, Distribution Weight=0.4, Evaluation Type=::NewEnumerator0, Recalculate Random Each Loop=true, Override Seed=false)
      Passthrough Float=&RandomRangeFloat(Minimum=0.135, Maximum=0.185)
     )
     B=&MergeHelper_FloatSwitch(0=0.18, 1=0.125, 2=0.125, 3=0.18, 4=0.18, Index=$User.ElimCount, Num=::NewEnumerator3)
    )
    UsePositionOffset=false
    Write Material Random=true
   )
   &ShapeLocation(
    MaskSpawnGroup=false
    Non Uniform Scale=&MaskVector(Mask Source=1, Mask Parameter=$User.ElimCount, Mask=0, Masked Vector=<2 1 2.2>, Passthrough Vector=<1 1 1>)
    Offset=&MergeHelper_VectorSwitch(0=<0 -5.25 6>, 1=<-1.9438 -5.25 6.5215>, 2=<-1.8215 -5.25 4.383>, 3=<0 -5.25 0>, 4=<0 -5.25 0>, Index=$User.ElimCount, Num=::NewEnumerator3)
    Offset Mode=::NewEnumerator0
    Override Seed=false
    Sphere Radius=&MergeHelper_FloatSwitch(0=4, 1=2, 2=2, 3=2, 4=2, Index=$User.ElimCount, Num=::NewEnumerator3)
    UseOwnerScale=true
   )
   &InitialMeshOrientation(Use Rotation Vector=false)
   &AddVelocity(
    Cone Angle=&MergeHelper_FloatSwitch(0=0, 1=25, 2=25, 3=45, 4=30, Index=$User.ElimCount, Num=::NewEnumerator3)
    Cone Axis=&MergeHelper_VectorSwitch(
     0=&RandomRangeVector(Minimum=<-0.15 0 0.7>, Maximum=<0.15 0 0.7>)
     1=<0.5 0 1>
     2=<-0.2 0 1>
     3=&RandomRangeVector(Minimum=<-0.15 0 0.7>, Maximum=<0.15 0 0.7>)
     4=&RandomRangeVector(Minimum=<-0.15 0 0.7>, Maximum=<0.15 0 0.7>)
     Index=$User.ElimCount
     Num=::NewEnumerator3
    )
    Module.UseConeFalloff=false
    Module.UseVelDistribution=false
    Override Seed=false
    Velocity Mode=::NewEnumerator2
    Velocity Speed=&Multiply_Float(
     A=&MaskFloatBySpawnGroup(
      Masked Float=&RandomRangeFloat(Minimum=100, Maximum=150)
      Passthrough Float=&RandomRangeFloat(Minimum=125, Maximum=180)
     )
     B=&MergeHelper_FloatSwitch(0=0, 1=0.8, 2=1, 3=1, 4=0.8, Index=$User.ElimCount, Num=::NewEnumerator3)
    )
   ) {module=AddVelocity001}
   &AddVelocity(
    Velocity=<0 0 1>
    Velocity Speed Scale=&MaskFloat(Mask Source=1, Mask Parameter=$User.ElimCount, Mask=0, Masked Float=0, Passthrough Float=80)
   ) {module=AddVelocity002}
   &AddVelocity(
    Velocity=<0 -1 0>
    Velocity Speed Scale=&MergeHelper_FloatSwitch(
     0=0
     1=80
     2=80
     3=&UniformAOrBFloat(A=80, B=100, Distribution Weight=0.2, Evaluation Type=::NewEnumerator0, Recalculate Random Each Loop=true, Override Seed=false)
     4=&UniformAOrBFloat(A=80, B=100, Distribution Weight=0.2, Evaluation Type=::NewEnumerator0, Recalculate Random Each Loop=true, Override Seed=false)
     Index=$User.ElimCount
     Num=::NewEnumerator3
    )
   )
   &AddVelocity(
    Velocity=<-1 0 0>
    Velocity Speed Scale=&MaskFloat(Mask Source=1, Mask Parameter=$User.ElimCount, Mask=0, Masked Float=0, Passthrough Float=20)
   ) {module=AddVelocity003}
   &AddVelocity(
    Module.Use Falloff Distance=false
    Override Seed=false
    Velocity Mode=::NewEnumerator1
    Velocity Origin=$Particles.ShapeLocation.ShapePosition
    Velocity Speed=&MaskFloat(
     Mask Source=1
     Mask Parameter=$User.ElimCount
     Mask=0
     Masked Float=&RandomRangeFloat(Minimum=65, Maximum=80)
     Passthrough Float=0
    )
   ) {module=AddVelocity004}
   &CameraOffset(Camera Offset Amount=10)
  ]
  particleUpdateScript [
   &ParticleState()
   &ScaleMeshSize(Scale Factor=&VectorFromFloat(Value=&FloatFromCurve(FloatCurve=DataInterface() {propertyValues {Curve {
         defaultValue=1
         keys [
          <at 0 0 RCIM_Cubic>
          {arriveTangentWeight=0.0916, interpMode=RCIM_Cubic, leaveTangentWeight=0.281, tangentMode=RCTM_User, time=0.1229, value=1}
          {arriveTangentWeight=0.1257, interpMode=RCIM_Cubic, leaveTangentWeight=0.1667, tangentMode=RCTM_User, time=0.5, value=1}
          {interpMode=RCIM_Cubic, tangentMode=RCTM_User, time=1, value=0}
         ]
        }}})))
   &GravityForce(Gravity=&MaskVector(Mask Source=1, Mask Parameter=$User.ElimCount, Mask=0, Masked Vector=<0 0 -200>, Passthrough Vector=<1 1 -980>))
   &MeshRotationForce(Rotation=&VectorFromFloat(Value=&RandomRangeFloat(Minimum=-120, Maximum=120)))
   &Drag(Drag=&RandomRangeFloat(Minimum=1, Maximum=3), Use Linear Drag=true, Use Rotational Drag=false)
   &SolveForcesAndVelocity(Clamp Velocity=false, Limit Acceleration=false)
  ]
  rendererClasses [&NiagaraMeshRendererProperties]
  renderers [
   {rendererClass=&NiagaraMeshRendererProperties, rendererIndex=0}
   {rendererClass=&NiagaraMeshRendererProperties, rendererIndex=1}
  ]
 }
 Droplets {
  emitterUpdateScript [
   &EmitterState()
   &SpawnBurst_Instantaneous(
    Spawn Count=&MaskInt(
     Mask Source:NiagaraInt32=1
     Mask Parameter:NiagaraInt32=$User.ElimCount
     Mask:NiagaraInt32=0
     Masked Int:NiagaraInt32=&RandomRangeInt(Minimum=12, Maximum=18)
     Passthrough Int:NiagaraInt32=&RandomRangeInt(Minimum=4, Maximum=6)
    )
    Use Spawn Probability=false
    UseLoopCountLimit=false
   )
  ]
  particleSpawnScript [
   &InitializeParticle(
    Color=&FNTimeOfDayColor(
     Night Color=<rgba 0.2835 0.4979 0.6979 1>
     Morning Color=<rgba 0.5854 0.7334 0.8714 1>
     Day Color=<rgba 0.5854 0.7334 0.8714 1>
     Evening Color=<rgba 0.5854 0.7334 0.8714 1>
    )
    Lifetime=&MaskFloat(
     Mask Source=1
     Mask Parameter=$User.ElimCount
     Mask=0
     Masked Float=&RandomRangeFloat(Minimum=0.275, Maximum=0.5)
     Passthrough Float=&RandomRangeFloat(Minimum=0.325, Maximum=0.6)
    )
    Sprite Size Mode=::NewEnumerator3
    Uniform Sprite Size=&Multiply_Float(
     A=&UniformAOrBFloat(A=0.05, B=0.125, Distribution Weight=0.4, Evaluation Type=::NewEnumerator0, Recalculate Random Each Loop=true, Override Seed=false)
     B=&MergeHelper_FloatSwitch(0=8, 1=5, 2=5, 3=8, 4=8, Index=$User.ElimCount, Num=::NewEnumerator3)
    )
    UsePositionOffset=false
    Write Material Random=true
   )
   &ShapeLocation(
    MaskSpawnGroup=false
    Non Uniform Scale=&MaskVector(Mask Source=1, Mask Parameter=$User.ElimCount, Mask=0, Masked Vector=<2 1 2.2>, Passthrough Vector=<1 1 1>)
    Offset=&MergeHelper_VectorSwitch(0=<0 -5.25 6>, 1=<-1.9438 -5.25 6.5215>, 2=<-1.8215 -5.25 4.383>, 3=<0 -5.25 0>, 4=<0 -5.25 0>, Index=$User.ElimCount, Num=::NewEnumerator3)
    Offset Mode=::NewEnumerator0
    Override Seed=false
    Sphere Radius=&MergeHelper_FloatSwitch(0=4, 1=2, 2=2, 3=2, 4=2, Index=$User.ElimCount, Num=::NewEnumerator3)
    UseOwnerScale=true
   )
   &InitialMeshOrientation(Use Rotation Vector=false)
   &AddVelocity(
    Cone Angle=&MergeHelper_FloatSwitch(0=30, 1=25, 2=25, 3=45, 4=30, Index=$User.ElimCount, Num=::NewEnumerator3)
    Cone Axis=&MergeHelper_VectorSwitch(
     0=&RandomRangeVector(Minimum=<-0.15 0 0.7>, Maximum=<0.15 0 0.7>)
     1=<0.5 0 1>
     2=<-0.2 0 1>
     3=&RandomRangeVector(Minimum=<-0.15 0 0.7>, Maximum=<0.15 0 0.7>)
     4=&RandomRangeVector(Minimum=<-0.15 0 0.7>, Maximum=<0.15 0 0.7>)
     Index=$User.ElimCount
     Num=::NewEnumerator3
    )
    Module.UseConeFalloff=false
    Module.UseVelDistribution=false
    Override Seed=false
    Velocity Mode=::NewEnumerator2
    Velocity Speed=&Multiply_Float(
     A=&RandomRangeFloat(Minimum=100, Maximum=150)
     B=&MergeHelper_FloatSwitch(0=0, 1=0.8, 2=1, 3=1, 4=0.8, Index=$User.ElimCount, Num=::NewEnumerator3)
    )
   ) {module=AddVelocity001}
   &AddVelocity(
    Velocity=<0 0 1>
    Velocity Speed Scale=&MaskFloat(Mask Source=1, Mask Parameter=$User.ElimCount, Mask=0, Masked Float=0, Passthrough Float=100)
   ) {module=AddVelocity002}
   &AddVelocity(
    Velocity=<0 -1 0>
    Velocity Speed Scale=&MergeHelper_FloatSwitch(
     0=0
     1=35
     2=35
     3=&UniformAOrBFloat(A=35, B=45, Distribution Weight=0.2, Evaluation Type=::NewEnumerator0, Recalculate Random Each Loop=true, Override Seed=false)
     4=&UniformAOrBFloat(A=35, B=45, Distribution Weight=0.2, Evaluation Type=::NewEnumerator0, Recalculate Random Each Loop=true, Override Seed=false)
     Index=$User.ElimCount
     Num=::NewEnumerator3
    )
   )
   &AddVelocity(
    Velocity=<-1 0 0>
    Velocity Speed Scale=&MaskFloat(Mask Source=1, Mask Parameter=$User.ElimCount, Mask=0, Masked Float=0, Passthrough Float=20)
   ) {module=AddVelocity003}
   &AddVelocity(
    Module.Use Falloff Distance=false
    Override Seed=false
    Velocity Mode=::NewEnumerator1
    Velocity Origin=$Particles.ShapeLocation.ShapePosition
    Velocity Speed=&MaskFloat(
     Mask Source=1
     Mask Parameter=$User.ElimCount
     Mask=0
     Masked Float=&RandomRangeFloat(Minimum=85, Maximum=100)
     Passthrough Float=0
    )
   ) {module=AddVelocity004}
   &CameraOffset(Camera Offset Amount=10)
   &SubUVAnimation(
    "Sprite Renderer "=DataInterface() {propertyValues {SpriteRenderer=&self:Droplets_0.NiagaraSpriteRendererProperties_0}}
    SubUV Animation Mode:ENiagara_SubUVLookupModeV2=::NewEnumerator2
    UseEndFrame=false
    UseStartFrame=false
   )
  ]
  particleUpdateScript [
   &ParticleState()
   &ScaleSpriteSize(Uniform Curve Sprite Scale=DataInterface() {propertyValues {Curve {
       defaultValue=1
       keys [
        <at 0 0 RCIM_Cubic>
        {arriveTangentWeight=0.0916, interpMode=RCIM_Cubic, leaveTangentWeight=0.281, tangentMode=RCTM_User, time=0.1229, value=1}
        {arriveTangentWeight=0.1257, interpMode=RCIM_Cubic, leaveTangentWeight=0.1667, tangentMode=RCTM_User, time=0.5, value=1}
        {interpMode=RCIM_Cubic, tangentMode=RCTM_User, time=1, value=0}
       ]
      }}})
   &GravityForce(Gravity=&MaskVector(
     Mask Source=1
     Mask Parameter=$User.ElimCount
     Mask=0
     Masked Vector=<0 0 -80>
     Passthrough Vector=&RandomRangeVector(Minimum=<0 0 -980>, Maximum=<0 0 -1100>)
    ))
   &Drag(Drag=&RandomRangeFloat(Minimum=0.8, Maximum=1.2), Use Linear Drag=true, Use Rotational Drag=false)
   &SolveForcesAndVelocity(Clamp Velocity=false, Limit Acceleration=false)
   &ScaleSpriteSizeBySpeed(Max Scale Factor=<0.3 4>, Velocity Threshold=250)
  ]
  rendererClasses [&NiagaraSpriteRendererProperties]
  renderers [{rendererClass=&NiagaraSpriteRendererProperties}]
 }
 Droplets001 {
  emitterUpdateScript [
   &EmitterState(Life Cycle Mode=::NewEnumerator1, Loop Duration=3, UseLoopDelay=false)
   &SpawnRate(
    Spawn Probability=&SelectFloatFromArray(Direct Array Index=$User.ElimCount, Float Selection Array=DataInterface() {propertyValues {FloatData [0,1,1.8,0,0]}})
    SpawnRate=2.2
    Use Spawn Probability=true
   )
  ]
  particleSpawnScript [
   &InitializeParticle(
    Color=&FNTimeOfDayColor(
     Night Color=<rgba 0.402 0.5012 0.5938 1>
     Morning Color=<rgba 0.5854 0.7334 0.8714 1>
     Day Color=<rgba 0.5854 0.7334 0.8714 1>
     Evening Color=<rgba 0.5854 0.7334 0.8714 1>
    )
    Lifetime Max=1
    Lifetime Min=0.6
    Lifetime Mode=::NewEnumerator1
    Sprite Size Mode=::NewEnumerator3
    Uniform Sprite Size=&UniformAOrBFloat(A=0.8, B=1.6, Distribution Weight=0.3, Evaluation Type=::NewEnumerator0, Recalculate Random Each Loop=true, Override Seed=false)
    UsePositionOffset=false
    Write Material Random=true
   )
   &ShapeLocation(
    Box Size=<8 2 8>
    MaskSpawnGroup=false
    Offset=&MergeHelper_VectorSwitch(0=<0 0 0>, 1=<-1.5591 -5.2515 5.3467>, 2=<-1.5591 -5.2515 2>, 3=<0 0 0>, 4=<0 0 0>, Index=$User.ElimCount, Num=::NewEnumerator3)
    Offset Mode=::NewEnumerator0
    Override Seed=false
    Shape Primitive=::NewEnumerator2
    Surface Only=true
    Surface Only Band Thickness=1
    UseOwnerScale=true
   )
   &InitialMeshOrientation(Use Rotation Vector=false)
   &CameraOffset(Camera Offset Amount=10)
   &SubUVAnimation(
    "Sprite Renderer "=DataInterface() {propertyValues {SpriteRenderer=&self:Droplets001_0.NiagaraSpriteRendererProperties_0}}
    SubUV Animation Mode:ENiagara_SubUVLookupModeV2=::NewEnumerator2
    UseEndFrame=false
    UseStartFrame=false
   )
  ]
  particleUpdateScript [
   &ParticleState()
   &ScaleSpriteSize(Uniform Curve Sprite Scale=DataInterface() {propertyValues {Curve {
       defaultValue=1
       keys [
        {arriveTangent=10.9911, interpMode=RCIM_Cubic, leaveTangent=10.9911, tangentMode=RCTM_User, time=0, value=0}
        {arriveTangentWeight=0.0916, interpMode=RCIM_Cubic, leaveTangentWeight=0.281, tangentMode=RCTM_User, time=0.1229, value=1}
        {interpMode=RCIM_Cubic, tangentMode=RCTM_User, time=1, value=0}
       ]
      }}})
   &GravityForce(Gravity=&MakeVector(
     X=&FloatFromCurve(
      FloatCurve=DataInterface() {propertyValues {Curve {keys [<at 0 1 RCIM_Constant>, {arriveTangent=-16.475, interpMode=RCIM_Constant, leaveTangent=-16.475, time=0.1, value=0}]}}}
      Scale Curve=-225
     )
     Y=0
     Z=&RandomRangeFloat(Minimum=-120, Maximum=-350)
    ))
   &Drag(Drag=&RandomRangeFloat(Minimum=5, Maximum=8), Use Linear Drag=true, Use Rotational Drag=false)
   &SolveForcesAndVelocity(Clamp Velocity=false, Limit Acceleration=false)
   &ScaleSpriteSizeBySpeed(Max Scale Factor=<0.3 4>, Velocity Threshold=80)
   &ScaleColor(Scale Alpha=&RandomRangeFloat(Minimum=0.4, Maximum=0.9), ScaleA=true, ScaleRGB=true)
  ]
  rendererClasses [&NiagaraSpriteRendererProperties]
  renderers [{rendererClass=&NiagaraSpriteRendererProperties}]
 }
 SmokeBlackDetail {
  emitterUpdateScript [
   &EmitterState(Life Cycle Mode=::NewEnumerator1, Loop Delay=0.4, Loop Duration=3, UseLoopDelay=false)
   &SpawnRate(
    Spawn Probability=&SelectFloatFromArray(Direct Array Index=$User.ElimCount, Float Selection Array=DataInterface() {propertyValues {FloatData [1,1,1,0,0]}})
    SpawnRate=3
    Use Spawn Probability=true
   )
   &SpawnBurst_Instantaneous(Use Spawn Probability=false, UseLoopCountLimit=false) {enabled=false}
  ]
  particleSpawnScript [
   &InitializeParticle(
    Color=&FNTimeOfDayColor(
     Night Color=<rgba 0.1343 0.2832 0.526 1>
     Morning Color=<rgba 0.4963 0.6463 0.8906 1>
     Day Color=<rgba 0.4963 0.6463 0.8906 1>
     Evening Color=<rgba 0.4963 0.6463 0.8906 1>
    )
    Lifetime=&MaskFloat(
     Mask Source=1
     Mask Parameter=$User.ElimCount
     Mask=2
     Masked Float=&RandomRangeFloat(Minimum=1.2, Maximum=2.4)
     Passthrough Float=&RandomRangeFloat(Minimum=1.6, Maximum=3.2)
    )
    Position Offset=<0 0 15>
    Sprite Rotation Mode=::NewEnumerator2
    Sprite Size Max=<8 15>
    Sprite Size Mode=::NewEnumerator2
    UsePositionOffset=true
    Write Material Random=true
   )
   &CameraOffset(Camera Offset Amount=&RandomRangeFloat(Minimum=5, Maximum=30))
   &ShapeLocation(
    Box Size=<2 4 8>
    MaskSpawnGroup=false
    Offset=&SelectVectorFromArray(
     Direct Array Index=$User.ElimCount
     Vector Selection Array=DataInterface() {propertyValues {FloatData [<0 -5.25 0>, <0 -5.25 0>, <0 -5.25 -5>, <0 0 0>, <0 0 0>]}}
    )
    Offset Mode=::NewEnumerator0
    Override Seed=false
    Shape Primitive=::NewEnumerator2
    Surface Only=false
    UseOwnerScale=true
   )
   &AddVelocity(
    Cone Axis=<0 0 1>
    Module.UseConeFalloff=false
    Module.UseVelDistribution=false
    Override Seed=false
    Velocity Mode=::NewEnumerator2
    Velocity Speed=&RandomRangeFloat(Minimum=-5, Maximum=-10)
   )
  ]
  particleUpdateScript [
   &ParticleState()
   &SpriteRotationRate(Rotation Rate=&RandomRangeFloat(Minimum=-50, Maximum=50))
   &ScaleSpriteSize(
    Scale Sprite Size Mode=::NewEnumerator0
    Uniform Scale Factor=&FloatFromCurve(FloatCurve=DataInterface() {propertyValues {Curve {keys [<at 0 0.8>,<at 1 1.2>]}}}, Scale Curve=1.3)
   )
   &SubUVAnimation(
    End Frame Range Override=30
    "Sprite Renderer "=DataInterface() {propertyValues {SpriteRenderer=&self:NiagaraEmitter_0.NiagaraSpriteRendererProperties_0}}
    Start Frame Range Override=&RandomRangeInt(Minimum=0, Maximum=5)
    UseEndFrame=true
    UseStartFrame=true
   )
   &ScaleColor(
    Scale Alpha=&Multiply_Float(
     A=&RandomRangeFloat(Minimum=0.3, Maximum=0.7)
     B=&FloatFromCurve(FloatCurve=DataInterface() {propertyValues {Curve {keys [
          {interpMode=RCIM_Cubic, tangentMode=RCTM_User, time=0, value=0}
          {interpMode=RCIM_Cubic, tangentMode=RCTM_User, time=0.5, value=1}
          {tangentMode=RCTM_User, time=1, value=0}
         ]}}})
    )
    ScaleA=true
    ScaleRGB=true
   )
   &SolveForcesAndVelocity(Clamp Velocity=false, Limit Acceleration=false)
  ]
  rendererClasses [&NiagaraSpriteRendererProperties]
  renderers [{rendererClass=&NiagaraSpriteRendererProperties}]
 }
 Smoke_Wisp {
  emitterUpdateScript [
   &EmitterState()
   &SpawnBurst_Instantaneous(
    Spawn Count=&MaskInt(
     Mask Source:NiagaraInt32=1
     Mask Parameter:NiagaraInt32=$User.ElimCount
     Mask:NiagaraInt32=0
     Masked Int:NiagaraInt32=&RandomRangeInt(Minimum=8, Maximum=10)
     Passthrough Int:NiagaraInt32=&RandomRangeInt(Minimum=2, Maximum=3)
    )
    Use Spawn Probability=false
    UseLoopCountLimit=false
   )
  ]
  particleSpawnScript [
   &InitializeParticle(
    Color=&FNTimeOfDayColor(
     Night Color=<rgba 0.2031 0.316 0.5 1>
     Morning Color=<rgba 0.4452 0.5421 0.7 1>
     Day Color=<rgba 0.4452 0.5421 0.7 1>
     Evening Color=<rgba 0.4452 0.5421 0.7 1>
    )
    Lifetime Max=0.8
    Lifetime Min=0.4
    Lifetime Mode=::NewEnumerator1
    Sprite Rotation Mode=::NewEnumerator2
    Sprite Size Mode=::NewEnumerator1
    Uniform Sprite Size Max=25
    Uniform Sprite Size Min=15
    UsePositionOffset=false
    Write Material Random=true
   )
   &ShapeLocation(
    MaskSpawnGroup=false
    Non Uniform Scale=&MaskVector(Mask Source=1, Mask Parameter=$User.ElimCount, Mask=0, Masked Vector=<1 1 2>, Passthrough Vector=<1 1 1>)
    Offset=&MergeHelper_VectorSwitch(0=<0 -5.25 6>, 1=<3.3539 -5.25 9.5789>, 2=<-1.8215 -5.25 4.383>, 3=<0 -5.25 0>, 4=<0 -5.25 0>, Index=$User.ElimCount, Num=::NewEnumerator3)
    Offset Mode=::NewEnumerator0
    Override Seed=false
    Sphere Radius=&MergeHelper_FloatSwitch(0=4, 1=2, 2=2, 3=2, 4=3, Index=$User.ElimCount, Num=::NewEnumerator3)
    UseOwnerScale=true
   )
   &AddVelocity(
    Cone Angle=&MergeHelper_FloatSwitch(0=30, 1=25, 2=25, 3=45, 4=30, Index=$User.ElimCount, Num=::NewEnumerator3)
    Cone Axis=&MergeHelper_VectorSwitch(0=<0 0 0>, 1=<1 0 1>, 2=<-0.2 0 1>, 3=<0 0 0.7>, 4=<0 0 0.7>, Index=$User.ElimCount, Num=::NewEnumerator3)
    Module.UseConeFalloff=false
    Module.UseVelDistribution=false
    Override Seed=false
    Velocity Mode=::NewEnumerator2
    Velocity Speed=&Multiply_Float(
     A=&RandomRangeFloat(Minimum=25, Maximum=40)
     B=&MergeHelper_FloatSwitch(0=0, 1=0.8, 2=1, 3=1, 4=1, Index=$User.ElimCount, Num=::NewEnumerator3)
    )
   ) {module=AddVelocity001}
   &AddVelocity(
    Velocity=<0 -1 0>
    Velocity Speed Scale=&MaskFloat(Mask Source=1, Mask Parameter=$User.ElimCount, Mask=0, Masked Float=0, Passthrough Float=45)
   )
   &AddVelocity(
    Velocity=<-1 0 0>
    Velocity Speed Scale=&MaskFloat(Mask Source=1, Mask Parameter=$User.ElimCount, Mask=0, Masked Float=0, Passthrough Float=20)
   ) {module=AddVelocity003}
   &AddVelocity(
    Module.Use Falloff Distance=false
    Override Seed=false
    Velocity Mode=::NewEnumerator1
    Velocity Origin=$Particles.ShapeLocation.ShapePosition
    Velocity Speed=&RandomRangeFloat(Minimum=15, Maximum=25)
   ) {module=AddVelocity002}
   &CameraOffset(Camera Offset Amount=25)
  ]
  particleUpdateScript [
   &ParticleState()
   &ScaleColor(
    Scale Alpha=&FloatFromCurve(
     FloatCurve=DataInterface() {propertyValues {Curve {keys [
         {interpMode=RCIM_Cubic, tangentMode=RCTM_User, time=0, value=0}
         {interpMode=RCIM_Cubic, tangentMode=RCTM_User, time=0.5, value=1}
         {tangentMode=RCTM_User, time=1, value=0}
        ]}}}
     Scale Curve=&MaskFloat(
      Mask Source=1
      Mask Parameter=$User.ElimCount
      Mask=0
      Masked Float=&RandomRangeFloat(Minimum=0.2, Maximum=0.45)
      Passthrough Float=&RandomRangeFloat(Minimum=0.35, Maximum=0.6)
     )
    )
    ScaleA=true
    ScaleRGB=true
   )
   &SpriteRotationRate()
   &ScaleSpriteSize(Scale Factor=&Vector2DFromFloat(Value=&FloatFromCurve(
      FloatCurve=DataInterface() {propertyValues {Curve {keys [
          {arriveTangent=2.4634, interpMode=RCIM_Cubic, leaveTangent=2.4634, tangentMode=RCTM_User, time=0, value=0.1}
          {arriveTangent=0.2754, interpMode=RCIM_Cubic, leaveTangent=0.2754, tangentMode=RCTM_User, time=1.0463, value=1}
         ]}}}
      Scale Curve=&MergeHelper_FloatSwitch(0=1, 1=1, 2=1, 3=0.7, 4=0.7, Index=$User.ElimCount, Num=::NewEnumerator3)
     )))
   &GravityForce(Gravity=&MakeVector(
     X=0
     Y=0
     Z=&MergeHelper_FloatSwitch(0=0, 1=-100, 2=-90, 3=-120, 4=-160, Index=$User.ElimCount, Num=::NewEnumerator3)
    ))
   &Drag(Drag=&RandomRangeFloat(Minimum=3, Maximum=5), Use Linear Drag=true, Use Rotational Drag=false)
   &SolveForcesAndVelocity(Clamp Velocity=false, Limit Acceleration=false)
   &SubUVAnimation(
    "Sprite Renderer "=DataInterface() {propertyValues {SpriteRenderer=&self:Smoke_Wisp_2.NiagaraSpriteRendererProperties_1}}
    UseEndFrame=false
    UseStartFrame=false
   )
  ]
  rendererClasses [&NiagaraSpriteRendererProperties]
  renderers [{rendererClass=&NiagaraSpriteRendererProperties}]
 }
}
summary {
 systemName=NS_Backpack_SpireTendFlair
 emitters [
  {emitterName=Chunk, rendererClasses [&NiagaraMeshRendererProperties]}
  {emitterName=Droplets, rendererClasses [&NiagaraSpriteRendererProperties]}
  {emitterName=Smoke_Wisp, rendererClasses [&NiagaraSpriteRendererProperties]}
  {emitterName=SmokeBlackDetail, rendererClasses [&NiagaraSpriteRendererProperties]}
  {emitterName=Droplets001, rendererClasses [&NiagaraSpriteRendererProperties]}
 ]
 userVariables [
  User.ElimCount {type=NiagaraInt32}
  User.VariantSwitch {type=NiagaraFloat, defaultValue=0, description=""}
 ]
}
