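# Anim Blueprint for the SpireTendFlair cosmetic backpack, which swaps its visible variant as its wearer eliminates enemy players.
# Spine-bend yaw drives an additive pack offset, then AnimDynamics springs jiggle two drop bones.
# Zero-scale bone modifiers, weighted by variant alphas, hide all but one of four variants.
# Elimination events step through them, hide all on the fourth kill, then reset after a delay; a looping timer cycles them in the front end.
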AnimBlueprint /BRCosmetics/Accessories/FORT_Backpacks/Backpack_M_MED_SpireTendFlair/Meshes/M_MED_SpireTendFlair_Pack_AnimBP
target_skeleton=&/M_MED_SpireTendFlair_Pack_Skeleton
graphs [
 AnimGraph {nodes [
   n1 = "Transform (Modify) Bone - Bone: pack"(ComponentPose=Local To Component(LocalPose=Input Pose()), Translation=Get backpack_translation(), Rotation=Get backpack_rotate()) {rotationMode=BMM_Additive, translationMode=BMM_Additive, alphaBoolBlend {blendOption=Linear}, boneToModify {boneName=pack}}
   n2 = Anim Dynamics - Bone: drop_V2!(SimulationSpace:AnimPhysSimSpaceType Enum, ComponentPose=@n1, Alpha=1, GravityOverride:Vector="0,0,0", ExternalForce="0,0,0") -> (Pose) {
    angularDampingOverride=0.9999
    angularSpringConstant=4000
    bAngularSpring=true
    bGravityOverrideInSimSpace=true
    bLinearSpring=true
    bOverrideAngularDamping=true
    bOverrideLinearDamping=true
    bUseGravityOverride=true
    linearDampingOverride=0.9999
    linearSpringConstant=8
    alphaBoolBlend {blendOption=Linear}
    boundBone {boneName=drop_V2}
    componentAppliedLinearAccClamp=<3000 3000 2500>
    componentLinearAccScale=<0.15 0.15 0.15>
    componentLinearVelScale=<0.15 0.15 0.15>
    physicsBodyDefinitions [{
      boundBone {boneName=drop_V2}
      boxExtents=<15 15 15>
      constraintSetup {
       coneAngle=1
       angularLimitsMax=<0 80 80>
       angularLimitsMin=<0 -80 -80>
       angularTarget=<1 0 0>
       linearAxesMax=<0.2 0 0>
       linearAxesMin=<-0.2 0 0>
      }
      localJointOffset=<1.6 0 0>
     }]
    simSpaceSettings {maxAngularAcceleration=35, maxAngularVelocity=7, simSpaceAngularAlpha=1}
   }
   n3 = Anim Dynamics - Bone: drop_V3!(SimulationSpace:AnimPhysSimSpaceType Enum, ComponentPose=@n2, Alpha=1, GravityOverride:Vector="0,0,0", ExternalForce="0,0,0") -> (Pose) {
    angularDampingOverride=0.9999
    angularSpringConstant=4000
    bAngularSpring=true
    bGravityOverrideInSimSpace=true
    bLinearSpring=true
    bOverrideAngularDamping=true
    bOverrideLinearDamping=true
    bUseGravityOverride=true
    linearDampingOverride=0.9999
    linearSpringConstant=8
    alphaBoolBlend {blendOption=Linear}
    boundBone {boneName=drop_V3}
    componentAppliedLinearAccClamp=<3000 3000 2500>
    componentLinearAccScale=<0.15 0.15 0.15>
    componentLinearVelScale=<0.15 0.15 0.15>
    physicsBodyDefinitions [{
      boundBone {boneName=drop_V3}
      boxExtents=<15 15 15>
      constraintSetup {
       coneAngle=1
       angularLimitsMax=<0 80 80>
       angularLimitsMin=<0 -80 -80>
       angularTarget=<1 0 0>
       linearAxesMax=<0.2 0 0>
       linearAxesMin=<-0.2 0 0>
      }
      localJointOffset=<1.6 0 0>
     }]
    simSpaceSettings {maxAngularAcceleration=35, maxAngularVelocity=7, simSpaceAngularAlpha=1}
   }
   n4 = "Transform (Modify) Bone - Bone: Variant_01"(ComponentPose=@n3, Alpha=Get Variant 01() -> (Variant 01:double), Translation="0,0,0", Rotation="0,0,0", Scale="0,0,0") {scaleMode=BMM_Additive, scaleSpace=BCS_BoneSpace, boneToModify {boneName=Variant_01}, scale=<0 0 0>}
   n5 = "Transform (Modify) Bone - Bone: Variant_02"(ComponentPose=@n4, Alpha=Get Variant 02() -> (Variant 02:double), Translation="0,0,0", Rotation="0,0,0", Scale="0,0,0") {scaleMode=BMM_Additive, scaleSpace=BCS_BoneSpace, boneToModify {boneName=Variant_02}, scale=<0 0 0>}
   n6 = "Transform (Modify) Bone - Bone: Variant_03"(ComponentPose=@n5, Alpha=Get Variant 03() -> (Variant 03:double), Translation="0,0,0", Rotation="0,0,0", Scale="0,0,0") {scaleMode=BMM_Additive, scaleSpace=BCS_BoneSpace, boneToModify {boneName=Variant_03}, scale=<0 0 0>}
   n7 = "Transform (Modify) Bone - Bone: Variant_04"(ComponentPose=@n6, Alpha=Get Variant 04() -> (Variant 04:double), Translation="0,0,0", Rotation="0,0,0", Scale="0,0,0") {scaleMode=BMM_Additive, scaleSpace=BCS_BoneSpace, boneToModify {boneName=Variant_04}, scale=<0 0 0>}
   Output Pose(Result=Component To Local(ComponentPose=@n7)) {layerGroup=None}
  ]}
 EventGraph {nodes [
   n1 = CustomEvent() {is=title K2Node_CustomEvent}
   > Branch(Condition=Get bIsInFrontEnd())
   > n2 = Increment Int(Value=Get KillCounter() -> (KillCounter:int=0))
   > n3 = Set KillCounter(KillCounter:int=@n2) -> (Output_Get:int=0)
   > n4 = Switch on Int!(execute:Exec, Selection:int=@n3) -> (Default:Exec, 0:Exec, 1:Exec, 2:Exec, 3:Exec, 4:Exec, 5:Exec)
   Set Variant 01(execute=@n4.1, Variant 01:double=1) -> (Output_Get:double)
   > Set Variant 02(Variant 02:double=0) -> (Output_Get:double)
   > Set Variant 03(Variant 03:double=1) -> (Output_Get:double)
   > Set Variant 04(Variant 04:double=1) -> (Output_Get:double)
   Set Variant 01(execute=@n4.2, Variant 01=1)
   > Set Variant 02(Variant 02=1)
   > Set Variant 03(Variant 03=0)
   > Set Variant 04(Variant 04=1)
   Set Variant 01(execute=@n4.3, Variant 01=1)
   > Set Variant 02(Variant 02=1)
   > Set Variant 03(Variant 03=1)
   > Set Variant 04(Variant 04=0)
   Set Variant 01(execute=@n4.4, Variant 01=1)
   > Set Variant 02(Variant 02=1)
   > Set Variant 03(Variant 03=1)
   > Set Variant 04(Variant 04=1)
   Set KillCounter(execute=@n4.5, KillCounter=0) -> (Output_Get=0)
   > Set Variant 01(Variant 01=0)
   > Set Variant 02(Variant 02=1)
   > Set Variant 03(Variant 03=1)
   > Set Variant 04(Variant 04=1)
   Event BlueprintUpdateAnimation()
   n5 = GetOwnerSkeletalMeshComponent()
   n6 = BreakRotator(InRot=BreakTransform(InTransform=GetDeltaTransformFromRefPose(self=@n5, BoneName=spine_05, BaseName=spine_01)).Rotation)
   > Set backpack_rotate(backpack_rotate=MakeRotator(Pitch=MapRangeClamped(Value=@n6.Yaw, InRangeA=-30, InRangeB=-150, OutRangeB=220)))
   > Set backpack_translation(backpack_translation=MakeVector(
     Y=MapRangeClamped(Value=@n6.Yaw, InRangeA=-20, InRangeB=-50, OutRangeB=-5)
     Z=MapRangeClamped(Value=@n6.Yaw, InRangeA=-30, InRangeB=-50, OutRangeB=3)
    ))
   n7 = Event BlueprintProcessAnimInputEvent()
   > Branch(Condition=IsNamed(self=@n7.AnimInputEvent, InEventName=aie.EnemyPlayerEliminated))
   > n8 = Increment Int(Value=Get KillCounter() -> (KillCounter=0))
   > n9 = Set KillCounter(KillCounter=@n8) -> (Output_Get=0)
   > n10 = Switch on Int!(execute, Selection=@n9) -> (Default, 0, 1, 2, 3, 4)
   Set Variant 01(execute=@n10.1, Variant 01=1)
   > Set Variant 02(Variant 02=0)
   > Set Variant 03(Variant 03=1)
   > Set Variant 04(Variant 04=1)
   Set Variant 01(execute=@n10.2, Variant 01=1)
   > Set Variant 02(Variant 02=1)
   > Set Variant 03(Variant 03=0)
   > Set Variant 04(Variant 04=1)
   Set Variant 01(execute=@n10.3, Variant 01=1)
   > Set Variant 02(Variant 02=1)
   > Set Variant 03(Variant 03=1)
   > Set Variant 04(Variant 04=0)
   Set Variant 01(execute=@n10.4, Variant 01=1)
   > Set Variant 02(Variant 02=1)
   > Set Variant 03(Variant 03=1)
   > Set Variant 04(Variant 04=1)
   > Delay(Duration=4)
   > Set KillCounter(KillCounter=0) -> (Output_Get=0)
   > Set Variant 01(Variant 01=0)
   > Set Variant 02(Variant 02=1)
   > Set Variant 03(Variant 03=1)
   > Set Variant 04(Variant 04=1)
   Event BlueprintBeginPlay() {is=title K2Node_Event}
   > Set Timer by Event(Delegate=@n1, Time=4, bLooping=true)
   TryGetPawnOwner()
  ]}
]
variables [
 backpack_rotate {type=Rotator}
 backpack_translation {type=Vector}
 KillCounter {type=integer}
 Variant 01 {type=number}
 Variant 02 {type=number, default=1}
 Variant 03 {type=number, default=1}
 Variant 04 {type=number, default=1}
]
